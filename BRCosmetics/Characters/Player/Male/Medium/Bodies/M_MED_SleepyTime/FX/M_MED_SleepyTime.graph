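# Translucent, unlit, two-sided FX material for the SleepyTime player body cosmetic, with SplineThicken width control.
# Body Mask01 channels blend colors and hue into SleepTimeFxFunction, whose result passes through emissive adjustments.
# The same channels select vertical panner speeds for Mask02, which is rounded to hard edges and multiplied by camera depth fade.
# A separate branch hue-shifts tinted, scrolling digital noise using rounded noise differences, ending in reroutes with no visible consumer.
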Material /BRCosmetics/Characters/Player/Male/Medium/Bodies/M_MED_SleepyTime/FX/M_MED_SleepyTime
graph {
 properties {blendMode=BLEND_Translucent, pixelDepthOffsetMode=PDOM_Legacy, shadingModel=MSM_Unlit, twoSided=true, usedShadingModels=Unlit}
 nodes [
  n1 = ScalarParameter() {defaultValue=5.4, parameterName=WPO_Size}
  MaterialFunctionCall(WidthBase=@n1, WidthTip=@n1) {materialFunction=&SplineThicken}
  n2 = TextureSample(UVs=TextureCoordinate()) {samplerType=SAMPLERTYPE_Masks, texture=/Game/Characters/Player/Male/Medium/Bodies/M_MED_SleepyTime/Textures/T_MED_SleepTime_Body_Mask01}
  n3 = LinearInterpolate(A=Constant() {r=1}, B=ScalarParameter() {defaultValue=0.4889, parameterName=Blue}, Alpha=@n2.G)
  n4 = ScalarParameter() {defaultValue=0.7, parameterName=purple}
  n5 = VectorParameter() {parameterName=magenta ColorA, defaultValue=<rgba 1 0 0.1111 1>}
  n6 = VectorParameter() {parameterName=insideMain, defaultValue=<rgba 1 0.7732 0.0645 1>}
  n7 = VectorParameter() {parameterName=InsideDark, defaultValue=<rgba 0.4097 0.1339 0.0433 1>}
  n8 = VectorParameter() {group=outlineCtrl, parameterName=outline color, defaultValue=<rgba 0.7656 0.6165 0.1979 1>}
  MaterialFunctionCall(Emissive=MaterialFunctionCall!(
    Exponen=ScalarParameter() {defaultValue=0.9324, group=outlineCtrl, parameterName=Exponen}
    Hue=LinearInterpolate(A=@n3, B=@n4, Alpha=@n2.B)
    InsideMaincolor=LinearInterpolate(A=@n5.RGB, B=@n6.RGB, Alpha=@n2.R)
    InsideDarkColor=LinearInterpolate(A=@n5.RGB, B=@n7.RGB, Alpha=@n2.R)
    outline color=LinearInterpolate(A=@n5.RGB, B=@n8.RGB, Alpha=@n2.R)
    BaseRef=ScalarParameter() {defaultValue=0.0643, group=outlineCtrl, parameterName=BaseRef}
   ) {materialFunction=/Game/Weapons/WeaponSkins/MaterialFunction/SleepTimeFxFunction}.Result) {materialFunction=&MF_FX_EmissiveAdjustments}
  n9 = LinearInterpolate(A=LinearInterpolate(A=Panner() {speedY=0.35}, B=Panner() {speedY=0.1}, Alpha=@n2.R), B=Panner() {speedY=0.7}, Alpha=@n2.B)
  n10 = Panner() {speedY=1.1}
  n11 = Constant() {r=0.3}
  n12 = Constant() {r=1}
  n13 = TextureSample(UVs=Multiply(A=LinearInterpolate(A=@n9, B=@n10, Alpha=@n2.G), B=AppendVector(A=@n11, B=@n12))) {samplerType=SAMPLERTYPE_Masks, texture=/Game/Characters/Player/Male/Medium/Bodies/M_MED_SleepyTime/Textures/T_MED_SleepTime_Body_Mask02}
  Multiply(
   A=MaterialFunctionCall(Fade Offset=ScalarParameter() {defaultValue=64, parameterName=CameraFadeOffset}) {materialFunction=&CameraDepthFade}.Result
   B=Round(None=@n13.R)
  )
  n14 = TextureCoordinate() {uTiling=0.25, vTiling=0.1}
  n15 = TextureSample(UVs=Panner(Coordinate=@n14) {speedY=1}) {texture=/Game/Effects/Fort_Effects/Effects/Atlas/Textures/T_Atlas_DigitalNoise_02}
  n16 = Subtract(
   A=@n15.R
   B=TextureSample(UVs=Panner(Coordinate=@n14) {speedY=0.4}) {texture=&T_Atlas_DigitalNoise_02}.B
  )
  n17 = Constant3Vector() {constant=<rgba 1 0.6321 0 0>}
  Reroute(None=MaterialFunctionCall(Hue Shift Percentage=Round(None=@n16), Texture=Multiply(A=@n15.RGB, B=@n17)) {materialFunction=&HueShift}.Result)
  Reroute(None=@n16)
  TextureSample() {samplerType=SAMPLERTYPE_Masks, texture=&T_MED_SleepTime_Body_Mask01}
 ]
}
