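# Niagara cosmetic effect for the male medium SleepyTime player body, built from a Flare emitter and a Light emitter.
# It exposes a bisHeadHidden user boolean, suggesting the effect responds to head visibility.
# The system scripts use SystemState, a spawn-time Set Variables assignment and a MakeCustomFloatFromBool dynamic input that turns a boolean into a float.
# Neither emitter lists any renderers here.
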NiagaraSystem /BRCosmetics/Characters/Player/Male/Medium/Bodies/M_MED_SleepyTime/FX/NS_MED_SleepyTime
dependencies {
 usedDataInterfaces []
 usedDynamicInputs [&MakeCustomFloatFromBool]
 usedModules [
  &SystemState
  &self:SystemSpawnScript.SystemScriptSource.SystemScriptGraph.NiagaraNodeAssignment_0.SetVariables_#1
 ]
}
emitters {Flare {rendererClasses [], renderers []}, Light {rendererClasses [], renderers []}}
summary {
 systemName=NS_MED_SleepyTime
 emitters [{emitterName=Flare, rendererClasses []}, {emitterName=Light, rendererClasses []}]
 userVariables [User.bisHeadHidden {type=NiagaraBool}]
}
