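# Material function that adds a two-tone emissive glow with an outline to masked areas of the SleepyTime player body cosmetic.
# An inner Fresnel blends dark and main colors; a second lerps in the outline color, both shaped by CheapContrast.
# The result drives emissive, alongside white base color and parameterized metallic, specular and roughness.
# These attributes blend over the input material using the max of two selectable Mask channels; a base-color-tinted multiply is left unconnected.
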MaterialFunction /BRCosmetics/Characters/Player/Male/Medium/Bodies/M_MED_SleepyTime/Materials/MF_SleepyTimeGlow
graph {nodes [
  Constant3Vector() {constant=<rgba 1 1 1 0>}
  n1 = GetMaterialAttributes(None=FunctionInput() {inputName=Material, inputType=FunctionInput_MaterialAttributes}) {attributeGetTypes [#1]}
  n2 = Fresnel(
   ExponentIn=ScalarParameter() {defaultValue=0.9091, group=SleepyTime_Glow, parameterName=ExponenInner}
   BaseReflectFractionIn=ScalarParameter() {defaultValue=0.0491, group=SleepyTime_Glow, parameterName=BaseRefInner}
  ) {exponent=2}
  n3 = ScalarParameter() {defaultValue=-0.019, group=SleepyTime_Glow, parameterName=inerlianeshape}
  n4 = ScalarParameter() {defaultValue=0.9324, group=SleepyTime_Glow, parameterName=Exponen}
  n5 = ScalarParameter() {defaultValue=0.0643, group=SleepyTime_Glow, parameterName=BaseRef}
  n6 = LinearInterpolate(
   A=LinearInterpolate(
    A=VectorParameter() {group=SleepyTime_Glow, parameterName=InsideDark, defaultValue=<rgba 0.4097 0.1339 0.0433 1>}.RGB
    B=VectorParameter() {group=SleepyTime_Glow, parameterName=insideMain, defaultValue=<rgba 1 0.7732 0.0645 1>}.RGB
    Alpha=MaterialFunctionCall(In=@n2, Contrast=@n3) {materialFunction=&CheapContrast}.Result
   )
   B=VectorParameter() {group=SleepyTime_Glow, parameterName=outline color, defaultValue=<rgba 0.7656 0.6165 0.1979 1>}.RGB
   Alpha=MaterialFunctionCall(
    In=Fresnel(ExponentIn=@n4, BaseReflectFractionIn=@n5) {exponent=2}
    Contrast=ScalarParameter() {defaultValue=11.3985, group=SleepyTime_Glow, parameterName=outlineShape}
   ) {materialFunction=&CheapContrast}.Result
  )
  n7 = SetMaterialAttributes!(
   MaterialAttributes=@n1.MaterialAttributes
   EmissiveColor=@n6
   Metallic=ScalarParameter() {group=SleepyTime, parameterName=SleepyTimeMetallic}
   Specular=ScalarParameter() {defaultValue=0.25, group=SleepyTime, parameterName=SleepyTimeSpecular}
   Roughness=ScalarParameter() {defaultValue=0.16, group=SleepyTime, parameterName=SleepyTimeRoughness}
   BaseColor=Constant3Vector()
  ) {attributeSetTypes [
    #2
    #3
    #4
    #5
    #1
   ]}
  n8 = FunctionInput(Preview=Constant4Vector() {constant=<rgba 1 1 1 1>}.RGBA) {inputName=Mask, inputType=FunctionInput_Vector4}
  n9 = ChannelMaskParameter(None=@n8) {group=SleepyTime_Glow, parameterName=ST_GlowMaskChannel}
  n10 = ChannelMaskParameter(None=@n8) {group=SleepyTime_Glow, maskChannel=Blue, parameterName=ST_GlowMaskChannel_02, defaultValue=<rgba 0 0 1 0>}
  FunctionOutput(None=BlendMaterialAttributes(A=@n1.MaterialAttributes, B=@n7, Alpha=Max(A=@n9.R, B=@n10.R)))
  Multiply(A=@n6, B=Clamp(None=Multiply(A=@n1.BaseColor) {constB=10}))
 ]}
