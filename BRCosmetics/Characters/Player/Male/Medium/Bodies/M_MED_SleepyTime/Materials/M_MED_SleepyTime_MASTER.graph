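# Master material for the SleepyTime medium male player character body, layering optional skin effects between shared character PreFX and PostFX functions.
# Static switches chain cloth fuzz, thin film, SleepyTime glow, cheap toon eyes and a flipbook mouth over material attributes.
# A SkinFX mask texture drives glow and, through selectable channels, masks fuzz, film and mouth.
# The shading model comes from the material expression, covering Default Lit, Subsurface Profile and Cloth.
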Material /BRCosmetics/Characters/Player/Male/Medium/Bodies/M_MED_SleepyTime/Materials/M_MED_SleepyTime_MASTER
graph {
 properties {
  bUseMaterialAttributes=true
  bUsedWithMeshDeformer=true
  bUsedWithMorphTargets=true
  bUsedWithSkeletalMesh=true
  pixelDepthOffsetMode=PDOM_Legacy
  shadingModel=MSM_FromMaterialExpression
  usedShadingModels=Default Lit | Subsurface Profile | Cloth
 }
 nodes [
  n1 = MaterialFunctionCall() {materialFunction=&BaseFN_CharacterFunction_PreFX}
  n2 = TextureSample(Tex=TextureObjectParameter() {group=Skin FX, parameterName=SkinFX_Mask})
  n3 = StaticComponentMaskParameter(None=@n2.RGBA) {defaultR=true, group=Skin FX, parameterName=Flipbook Texture Channel}
  n4 = StaticSwitchParameter(
   True=MaterialFunctionCall!(
    Alpha=StaticComponentMaskParameter(None=@n2.RGBA) {defaultR=true, group=Skin FX, parameterName=Cloth Channel}
    Gmap=@n1.Gmap
    Material=@n1.Gmap
   ) {materialFunction=/Game/Weapons/WeaponSkins/MaterialFunction/ClothFuzz}.Result
   False=@n1.Result
  ) {group=Skin FX, parameterName=Use Cloth fuzz}
  n5 = StaticSwitchParameter(
   True=MaterialFunctionCall!(Alpha=StaticComponentMaskParameter(None=@n2.RGBA) {defaultR=true, group=Skin FX, parameterName=ThinFilm_Channel}, Material Attributes=@n4) {materialFunction=/Game/Weapons/WeaponSkins/MaterialFunction/ThinFilmFunction}.Result
   False=@n4
  ) {group=Skin FX, parameterName=Use Thin Film}
  n6 = StaticSwitchParameter(
   True=MaterialFunctionCall!(Mask=@n2.RGBA, Material=@n5) {materialFunction=&/MF_SleepyTimeGlow}.Result
   False=@n5
  ) {parameterName=UseSleepyTimeGlow}
  n7 = StaticSwitchParameter(
   True=MaterialFunctionCall!(Material=@n6) {materialFunction=/Game/Characters/Player/Male/Medium/Bodies/M_MED_SleepyTime/Materials/MF_CheapToonEyes}.Result
   False=@n6
  ) {parameterName=UseCheapToonEyes}
  MaterialFunctionCall(
   AO=@n1.AO
   Distance Blend=@n1.AO
   Fresnel=@n1.AO
   Gmap=@n1.AO
   Pre Skinned Local Position=@n1.AO
   Material=StaticSwitchParameter(
    True=MaterialFunctionCall!(Alpha=@n3, Material_1=@n7) {materialFunction=/Game/Characters/Player/Male/Medium/Bodies/M_MED_SleepyTime/Materials/MF_CheapFlipBookMouth}.Result
    False=@n7
   ) {group=Skin FX, parameterName=Use Sub UV texture}
  ) {materialFunction=&BaseFN_CharacterFunction_PostFX}
 ]
}
