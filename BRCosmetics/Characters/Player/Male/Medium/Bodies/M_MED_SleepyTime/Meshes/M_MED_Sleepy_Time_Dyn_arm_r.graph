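# Physics asset for the right-arm shoulder pad dynamic bones on the medium male SleepyTime body cosmetic.
# Three box bodies form the pads, with capsules for spine_05 and the right upper-arm twist bone.
# The first pad hangs off the upper arm and carries the second and back pads, with tight twist limits and a wider swing for the back.
# The arm-to-spine constraint leaves swing and twist free.
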PhysicsAsset /BRCosmetics/Characters/Player/Male/Medium/Bodies/M_MED_SleepyTime/Meshes/M_MED_Sleepy_Time_Dyn_arm_r
bodies {
 dyn_shoulderpad_bk_r_01 [{shapeType=Box, extentX=3, extentY=13, extentZ=14, center=<6 0 0>, rotation=<rot 90 0 0>}]
 dyn_shoulderpad_r_01 [{shapeType=Box, extentX=3.5, extentY=9.5, extentZ=14, center=<6 0 0>, rotation=<rot 90 0 0>}]
 dyn_shoulderpad_r_02 [{shapeType=Box, extentX=3.5, extentY=9.5, extentZ=14, center=<6 0 0>, rotation=<rot 90 0 0>}]
 spine_05 [{shapeType=Capsule, length=30, radius=10}]
 upperarm_twist_01_r [{shapeType=Capsule, length=22, radius=4.5, center=<-14 0 0>, rotation=<rot 90 0 0>}]
}
constraints [
 {bone1Name=dyn_shoulderpad_r_01, bone2Name=upperarm_twist_01_r, swing1LimitDegrees=10, swing2LimitDegrees=20, twistLimitDegrees=1}
 {bone1Name=dyn_shoulderpad_r_02, bone2Name=dyn_shoulderpad_r_01, swing1LimitDegrees=10, swing2LimitDegrees=20, twistLimitDegrees=1}
 {bone1Name=dyn_shoulderpad_bk_r_01, bone2Name=dyn_shoulderpad_r_01, swing1LimitDegrees=30, swing2LimitDegrees=60, twistLimitDegrees=1}
 {
  bone1Name=upperarm_twist_01_r
  bone2Name=spine_05
  swing1LimitDegrees=45
  swing1Motion=Free
  swing2LimitDegrees=45
  swing2Motion=Free
  twistLimitDegrees=45
  twistMotion=Free
 }
]
