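# Physics asset for the pillow collar of the SleepyTime medium male body cosmetic, defining collision and joints for its dynamic bones.
# Six box bodies, three per side, cover the pillow bones, each joined to its own tiny MSTR capsule with narrow angular limits, tightest on right pillow 04.
# The MSTR capsules and a large head capsule connect to a small spine_05 capsule through constraints with free swing and twist.
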PhysicsAsset /BRCosmetics/Characters/Player/Male/Medium/Bodies/M_MED_SleepyTime/Meshes/M_MED_Sleepy_Time_Dyn_collar
bodies {
 dyn_head [{shapeType=Capsule, length=7.5, radius=10, rotation=<rot 90 0 0>}]
 dyn_pillow_l_01 [{shapeType=Box, extentX=20, extentY=10, extentZ=4, center=<9 0 0>, rotation=<rot 0 0 -25>}]
 dyn_pillow_l_01_MSTR [{shapeType=Capsule, length=0.5, radius=0.5, rotation=<rot 0 -90 0>}]
 dyn_pillow_l_02 [{shapeType=Box, extentX=17, extentY=10, extentZ=6.5, center=<7 0 0>, rotation=<rot 0 0 10>}]
 dyn_pillow_l_02_MSTR [{shapeType=Capsule, length=0.5, radius=0.5, rotation=<rot 0 -90 0>}]
 dyn_pillow_l_03 [{shapeType=Box, extentX=17, extentY=10, extentZ=8.5, center=<7 0 0>}]
 dyn_pillow_l_03_MSTR [{shapeType=Capsule, length=0.5, radius=0.5, rotation=<rot 0 -90 0>}]
 dyn_pillow_r_02 [{shapeType=Box, extentX=16.5, extentY=24, extentZ=5, center=<12 0 0>}]
 dyn_pillow_r_02_MSTR [{shapeType=Capsule, length=0.5, radius=0.5, rotation=<rot 0 -90 0>}]
 dyn_pillow_r_03 [{shapeType=Box, extentX=17.5, extentY=22, extentZ=5, center=<9 0 0>, rotation=<rot 0 0 -10.0008>}]
 dyn_pillow_r_03_MSTR [{shapeType=Capsule, length=0.5, radius=0.5, rotation=<rot 0 -90 0>}]
 dyn_pillow_r_04 [{shapeType=Box, extentX=17, extentY=12, extentZ=5.5, center=<8.8672 0 0.6779>, rotation=<rot 14.9999 0 0>}]
 dyn_pillow_r_04_MSTR [{shapeType=Capsule, length=0.5, radius=0.5, rotation=<rot 0 -90 0>}]
 spine_05 [{shapeType=Capsule, length=2, radius=2}]
}
constraints [
 {bone1Name=dyn_pillow_l_01, bone2Name=dyn_pillow_l_01_MSTR, swing1LimitDegrees=15, swing2LimitDegrees=45, twistLimitDegrees=5}
 {bone1Name=dyn_pillow_l_02, bone2Name=dyn_pillow_l_02_MSTR, swing1LimitDegrees=15, swing2LimitDegrees=45, twistLimitDegrees=5}
 {bone1Name=dyn_pillow_l_03, bone2Name=dyn_pillow_l_03_MSTR, swing1LimitDegrees=15, swing2LimitDegrees=45, twistLimitDegrees=5}
 {
  bone1Name=dyn_head
  bone2Name=spine_05
  swing1LimitDegrees=45
  swing1Motion=Free
  swing2LimitDegrees=45
  swing2Motion=Free
  twistLimitDegrees=45
  twistMotion=Free
 }
 {bone1Name=dyn_pillow_r_03, bone2Name=dyn_pillow_r_03_MSTR, swing1LimitDegrees=15, swing2LimitDegrees=45, twistLimitDegrees=5}
 {bone1Name=dyn_pillow_r_02, bone2Name=dyn_pillow_r_02_MSTR, swing1LimitDegrees=15, swing2LimitDegrees=45, twistLimitDegrees=5}
 {bone1Name=dyn_pillow_r_04, bone2Name=dyn_pillow_r_04_MSTR, swing1LimitDegrees=2, swing2LimitDegrees=30, twistLimitDegrees=2}
 {
  bone1Name=dyn_pillow_r_04_MSTR
  bone2Name=spine_05
  swing1LimitDegrees=45
  swing1Motion=Free
  swing2LimitDegrees=45
  swing2Motion=Free
  twistLimitDegrees=45
  twistMotion=Free
 }
 {
  bone1Name=dyn_pillow_r_03_MSTR
  bone2Name=spine_05
  swing1LimitDegrees=45
  swing1Motion=Free
  swing2LimitDegrees=45
  swing2Motion=Free
  twistLimitDegrees=45
  twistMotion=Free
 }
 {
  bone1Name=dyn_pillow_r_02_MSTR
  bone2Name=spine_05
  swing1LimitDegrees=45
  swing1Motion=Free
  swing2LimitDegrees=45
  swing2Motion=Free
  twistLimitDegrees=45
  twistMotion=Free
 }
 {
  bone1Name=dyn_pillow_l_03_MSTR
  bone2Name=spine_05
  swing1LimitDegrees=45
  swing1Motion=Free
  swing2LimitDegrees=45
  swing2Motion=Free
  twistLimitDegrees=45
  twistMotion=Free
 }
 {
  bone1Name=dyn_pillow_l_02_MSTR
  bone2Name=spine_05
  swing1LimitDegrees=45
  swing1Motion=Free
  swing2LimitDegrees=45
  swing2Motion=Free
  twistLimitDegrees=45
  twistMotion=Free
 }
 {
  bone1Name=dyn_pillow_l_01_MSTR
  bone2Name=spine_05
  swing1LimitDegrees=45
  swing1Motion=Free
  swing2LimitDegrees=45
  swing2Motion=Free
  twistLimitDegrees=45
  twistMotion=Free
 }
]
