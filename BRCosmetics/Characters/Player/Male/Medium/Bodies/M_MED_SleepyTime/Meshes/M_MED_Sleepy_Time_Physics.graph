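# Physics asset for the male medium SleepyTime body cosmetic, setting up ragdoll-style bodies for its coat around the hips and legs.
# Six chains of box bodies hang from the thigh twist bones with tight angular limits, with extra angled boxes curving the upper side panels.
# Cross constraints tie the last back segment to the matching side segment on each side.
# Thigh and calf capsules join the pelvis with free angular motion.
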PhysicsAsset /BRCosmetics/Characters/Player/Male/Medium/Bodies/M_MED_SleepyTime/Meshes/M_MED_Sleepy_Time_Physics
bodies {
 calf_l [{shapeType=Capsule, length=47, radius=5, center=<-25 -3.3 0>, rotation=<rot 90 0 3.6>}]
 calf_r [{shapeType=Capsule, length=29, radius=5, center=<17 3.8 0>, rotation=<rot 90 0 3.6>}]
 dyn_bk_l_coat_01 [{shapeType=Box, extentX=13, extentY=16, extentZ=4, center=<3.5 0 0>}]
 dyn_bk_l_coat_02 [{shapeType=Box, extentX=13, extentY=16, extentZ=5.318, center=<3.5 0 0>}]
 dyn_bk_l_coat_03 [{shapeType=Box, extentX=13, extentY=20, extentZ=6, center=<3.5 0 0>}]
 dyn_bk_l_coat_04 [{shapeType=Box, extentX=13, extentY=20, extentZ=6, center=<3.5 0 0>}]
 dyn_bk_r_coat_01 [{shapeType=Box, extentX=13, extentY=16, extentZ=4, center=<3.5 0 0>}]
 dyn_bk_r_coat_02 [{shapeType=Box, extentX=13, extentY=16, extentZ=5.318, center=<3.5 0 0>}]
 dyn_bk_r_coat_03 [{shapeType=Box, extentX=13, extentY=16, extentZ=5.318, center=<3.5 0 0>}]
 dyn_bk_r_coat_04 [{shapeType=Box, extentX=10, extentY=13, extentZ=5.3, center=<2 0 0>}]
 dyn_l_bk_coat_01 [{shapeType=Box, extentX=13, extentY=20, extentZ=4, center=<5 0 0>}]
 dyn_l_bk_coat_02 [{shapeType=Box, extentX=13, extentY=20, extentZ=4, center=<5 0 0>}]
 dyn_l_bk_coat_03 [{shapeType=Box, extentX=13, extentY=20, extentZ=4, center=<5 0 0>}]
 dyn_l_coat_01 [
  {shapeType=Box, extentX=13, extentY=12, extentZ=4.8406, center=<3.5 0 0>}
  {shapeType=Box, extentX=13, extentY=9, extentZ=3.5, center=<2.5 8.5 -3>, rotation=<rot -3.5 3.5 45>}
  {shapeType=Box, extentX=13, extentY=10, extentZ=3.5, center=<1 12 -11>, rotation=<rot -9.5 6 85>}
 ]
 dyn_l_coat_02 [
  {shapeType=Box, extentX=13, extentY=12, extentZ=4.8406, center=<3.5 0 0>}
  {shapeType=Box, extentX=13, extentY=9, extentZ=4.5, center=<3 9.5 -4>, rotation=<rot -7 7 45>}
  {shapeType=Box, extentX=13, extentY=10, extentZ=5.5, center=<2 12 -11>, rotation=<rot -10 0 85>}
 ]
 dyn_l_coat_03 [
  {shapeType=Box, extentX=13, extentY=17, extentZ=7, center=<3.5 0 0>}
  {shapeType=Box, extentX=13, extentY=10, extentZ=5.5, center=<2.8 12 -5.5>, rotation=<rot -4 3.2 50>}
 ]
 dyn_l_coat_04 [{shapeType=Box, extentX=13, extentY=17, extentZ=6, center=<3.5 0 0>}]
 dyn_r_bk_coat_01 [{shapeType=Box, extentX=13, extentY=20, extentZ=4, center=<5 0 0>}]
 dyn_r_bk_coat_02 [{shapeType=Box, extentX=13, extentY=20, extentZ=4, center=<5 0 0>}]
 dyn_r_bk_coat_03 [{shapeType=Box, extentX=13, extentY=20, extentZ=4, center=<5 0 0>}]
 dyn_r_coat_01 [
  {shapeType=Box, extentX=13, extentY=12, extentZ=4.8406, center=<3.5 0 0>}
  {shapeType=Box, extentX=13, extentY=9, extentZ=3.5, center=<2.5 -8.5 -3>, rotation=<rot -3.5 -3.5 -45>}
  {shapeType=Box, extentX=13, extentY=10, extentZ=3.5, center=<1 -12 -11>, rotation=<rot -9.5 -6 -85>}
 ]
 dyn_r_coat_02 [
  {shapeType=Box, extentX=13, extentY=12, extentZ=4.8406, center=<3.5 0 0>}
  {shapeType=Box, extentX=13, extentY=9, extentZ=3.5, center=<3 -9.5 -4>, rotation=<rot -7 -7 -45>}
  {shapeType=Box, extentX=13, extentY=10, extentZ=3.5, center=<2 -12 -11>, rotation=<rot -10 0 -85>}
 ]
 dyn_r_coat_03 [
  {shapeType=Box, extentX=13, extentY=17, extentZ=6, center=<3.5 0 0>}
  {shapeType=Box, extentX=13, extentY=10, extentZ=3.5, center=<2.8 -12 -5.5>, rotation=<rot -4 -3.2 -50>}
 ]
 dyn_r_coat_04 [{shapeType=Box, extentX=13, extentY=17, extentZ=6, center=<3.5 0 0>}]
 pelvis [{shapeType=Capsule, length=2, radius=2}]
 thigh_twist_01_l [
  {shapeType=Capsule, length=12, radius=8, center=<14 -0.75 1>, rotation=<rot 90 0 0>}
  {shapeType=Capsule, length=14, radius=6.5, center=<-9.2718 0.3745 0>, rotation=<rot 90 90 85.7158>}
 ]
 thigh_twist_01_r [
  {shapeType=Capsule, length=12, radius=8, center=<-14 0.75 1>, rotation=<rot 90 0 0>}
  {shapeType=Capsule, length=12.5, radius=6.5, center=<6 1 0>, rotation=<rot 90 0 0>}
 ]
}
constraints [
 {bone1Name=dyn_l_coat_02, bone2Name=dyn_l_coat_01, swing1LimitDegrees=5, swing2LimitDegrees=30, twistLimitDegrees=5}
 {bone1Name=dyn_l_coat_03, bone2Name=dyn_l_coat_02, swing1LimitDegrees=5, swing2LimitDegrees=30, twistLimitDegrees=5}
 {bone1Name=dyn_l_coat_04, bone2Name=dyn_l_coat_03, swing1LimitDegrees=5, swing2LimitDegrees=30, twistLimitDegrees=5}
 {bone1Name=dyn_l_bk_coat_02, bone2Name=dyn_l_bk_coat_01, swing1LimitDegrees=5, swing2LimitDegrees=30, twistLimitDegrees=5}
 {bone1Name=dyn_l_bk_coat_03, bone2Name=dyn_l_bk_coat_02, swing1LimitDegrees=5, swing2LimitDegrees=30, twistLimitDegrees=5}
 {bone1Name=dyn_r_coat_02, bone2Name=dyn_r_coat_01, swing1LimitDegrees=5, swing2LimitDegrees=30, twistLimitDegrees=5}
 {bone1Name=dyn_r_coat_03, bone2Name=dyn_r_coat_02, swing1LimitDegrees=5, swing2LimitDegrees=30, twistLimitDegrees=5}
 {bone1Name=dyn_r_coat_04, bone2Name=dyn_r_coat_03, swing1LimitDegrees=5, swing2LimitDegrees=30, twistLimitDegrees=5}
 {bone1Name=dyn_r_bk_coat_02, bone2Name=dyn_r_bk_coat_01, swing1LimitDegrees=5, swing2LimitDegrees=30, twistLimitDegrees=5}
 {bone1Name=dyn_r_bk_coat_03, bone2Name=dyn_r_bk_coat_02, swing1LimitDegrees=5, swing2LimitDegrees=30, twistLimitDegrees=5}
 {bone1Name=dyn_bk_l_coat_02, bone2Name=dyn_bk_l_coat_01, swing1LimitDegrees=5, swing2LimitDegrees=30, twistLimitDegrees=5}
 {bone1Name=dyn_bk_l_coat_03, bone2Name=dyn_bk_l_coat_02, swing1LimitDegrees=5, swing2LimitDegrees=30, twistLimitDegrees=5}
 {bone1Name=dyn_bk_l_coat_04, bone2Name=dyn_bk_l_coat_03, swing1LimitDegrees=5, swing2LimitDegrees=30, twistLimitDegrees=5}
 {bone1Name=dyn_bk_r_coat_02, bone2Name=dyn_bk_r_coat_01, swing1LimitDegrees=5, swing2LimitDegrees=30, twistLimitDegrees=5}
 {bone1Name=dyn_bk_r_coat_03, bone2Name=dyn_bk_r_coat_02, swing1LimitDegrees=5, swing2LimitDegrees=30, twistLimitDegrees=5}
 {bone1Name=dyn_bk_r_coat_04, bone2Name=dyn_bk_r_coat_03, swing1LimitDegrees=5, swing2LimitDegrees=30, twistLimitDegrees=5}
 {bone1Name=dyn_r_coat_01, bone2Name=thigh_twist_01_r, swing1LimitDegrees=5, swing2LimitDegrees=30, twistLimitDegrees=5}
 {bone1Name=dyn_bk_r_coat_01, bone2Name=thigh_twist_01_r, swing1LimitDegrees=5, swing2LimitDegrees=30, twistLimitDegrees=5}
 {bone1Name=dyn_r_bk_coat_01, bone2Name=thigh_twist_01_r, swing1LimitDegrees=5, swing2LimitDegrees=30, twistLimitDegrees=5}
 {bone1Name=dyn_bk_l_coat_04, bone2Name=dyn_l_coat_04, swing1LimitDegrees=60, swing2LimitDegrees=30, twistLimitDegrees=5}
 {bone1Name=dyn_l_coat_01, bone2Name=thigh_twist_01_l, swing1LimitDegrees=5, swing2LimitDegrees=30, twistLimitDegrees=5}
 {bone1Name=dyn_bk_l_coat_01, bone2Name=thigh_twist_01_l, swing1LimitDegrees=5, swing2LimitDegrees=30, twistLimitDegrees=5}
 {bone1Name=dyn_l_bk_coat_01, bone2Name=thigh_twist_01_l, swing1LimitDegrees=5, swing2LimitDegrees=30, twistLimitDegrees=5}
 {
  bone1Name=calf_l
  bone2Name=pelvis
  swing1LimitDegrees=45
  swing1Motion=Free
  swing2LimitDegrees=45
  swing2Motion=Free
  twistLimitDegrees=45
  twistMotion=Free
 }
 {
  bone1Name=thigh_twist_01_l
  bone2Name=pelvis
  swing1LimitDegrees=45
  swing1Motion=Free
  swing2LimitDegrees=45
  swing2Motion=Free
  twistLimitDegrees=45
  twistMotion=Free
 }
 {
  bone1Name=thigh_twist_01_r
  bone2Name=pelvis
  swing1LimitDegrees=45
  swing1Motion=Free
  swing2LimitDegrees=45
  swing2Motion=Free
  twistLimitDegrees=45
  twistMotion=Free
 }
 {
  bone1Name=calf_r
  bone2Name=pelvis
  swing1LimitDegrees=45
  swing1Motion=Free
  swing2LimitDegrees=45
  swing2Motion=Free
  twistLimitDegrees=45
  twistMotion=Free
 }
 {bone1Name=dyn_bk_r_coat_04, bone2Name=dyn_r_coat_04, swing1LimitDegrees=75, swing2LimitDegrees=15, twistLimitDegrees=10}
]
