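# Post-copy AnimBlueprint for the SleepyTime male body cosmetic, adding secondary motion to pillows, coat and a floating head.
# It copies the parent mesh pose, constrains pillow bones to spine and clavicles, adds sine head drift, then runs four RigidBody simulations.
# Additive bone correctives follow, driven by event-graph alphas from arm angles, parachute, jumping and crouching.
# Movement states also soften or reset physics; a second, weighted constraint chain is left disconnected.
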AnimBlueprint /BRCosmetics/Characters/Player/Male/Medium/Bodies/M_MED_SleepyTime/Meshes/M_MED_Sleepy_Time_AnimBP
target_skeleton=&/M_MED_Sleepy_Time_Skeleton
graphs [
 AnimGraph {nodes [
   n1 = "Constraint - dyn_pillow_l_01_MSTR (PP)"!(ComponentPose, Alpha=1, ConstraintWeights_0=0.5, ConstraintWeights_1:float=0.5) -> (Pose) {
    alphaBoolBlend {blendOption=Linear}
    boneToModify {boneName=dyn_pillow_l_01_MSTR}
    constraintSetup [{targetBone {boneName=spine_05}}, {targetBone {boneName=clavicle_l}}]
    constraintWeights [1,1]
   }
   n2 = "Constraint - dyn_pillow_l_02_MSTR (PPP)"!(ComponentPose=@n1, Alpha=1, ConstraintWeights_0=0.15, ConstraintWeights_1:float=0.72, ConstraintWeights_2:float=0.13) -> (Pose) {
    alphaBoolBlend {blendOption=Linear}
    boneToModify {boneName=dyn_pillow_l_02_MSTR}
    constraintSetup [{targetBone {boneName=spine_05}}, {targetBone {boneName=clavicle_l}}, {targetBone {boneName=upperarm_twist_01_l}}]
    constraintWeights [1,1,1]
   }
   n3 = "Constraint - dyn_pillow_l_03_MSTR (PPP)"!(ComponentPose=@n2, Alpha=1, ConstraintWeights_0=0.41, ConstraintWeights_1:float=0.45, ConstraintWeights_2:float=0.14) -> (Pose) {
    alphaBoolBlend {blendOption=Linear}
    boneToModify {boneName=dyn_pillow_l_03_MSTR}
    constraintSetup [{targetBone {boneName=spine_05}}, {targetBone {boneName=clavicle_l}}, {targetBone {boneName=upperarm_twist_01_l}}]
    constraintWeights [1,1,1]
   }
   n4 = "Constraint - dyn_pillow_l_03_MSTR (PPPP)"!(
    ComponentPose=@n3
    Alpha=1
    ConstraintWeights_0=0.63
    ConstraintWeights_1:float=0.27
    ConstraintWeights_2:float=0.05
    ConstraintWeights_3:float=0.05
   ) -> (Pose) {
    alphaBoolBlend {blendOption=Linear}
    boneToModify {boneName=dyn_pillow_l_03_MSTR}
    constraintSetup [
     {targetBone {boneName=spine_05}}
     {targetBone {boneName=clavicle_l}}
     {targetBone {boneName=clavicle_r}}
     {targetBone {boneName=upperarm_twist_01_l}}
    ]
    constraintWeights [1,1,1,1]
   }
   n5 = "Constraint - dyn_pillow_r_00_MSTR (PPPP)"!(
    ComponentPose=@n4
    Alpha=1
    ConstraintWeights_0=0.42
    ConstraintWeights_1:float=0.45
    ConstraintWeights_2:float=0.03
    ConstraintWeights_3:float=0.1
   ) -> (Pose) {
    alphaBoolBlend {blendOption=Linear}
    boneToModify {boneName=dyn_pillow_r_00_MSTR}
    constraintSetup [{targetBone {boneName=spine_04}}, {targetBone {boneName=spine_05}}, {targetBone {boneName=clavicle_r}}, {targetBone {boneName=clavicle_l}}]
    constraintWeights [1,1,1,1]
   }
   n6 = "Constraint - dyn_pillow_r_up_01_MSTR (PPPP)"!(
    ComponentPose=@n5
    Alpha=1
    ConstraintWeights_0=0.31
    ConstraintWeights_1:float=0.39
    ConstraintWeights_2:float=0.1
    ConstraintWeights_3:float=0.2
   ) -> (Pose) {
    alphaBoolBlend {blendOption=Linear}
    boneToModify {boneName=dyn_pillow_r_up_01_MSTR}
    constraintSetup [{targetBone {boneName=spine_04}}, {targetBone {boneName=spine_05}}, {targetBone {boneName=clavicle_r}}, {targetBone {boneName=clavicle_l}}]
    constraintWeights [1,1,1,1]
   }
   n7 = "Constraint - dyn_pillow_r_up_02_MSTR (PPPP)"!(
    ComponentPose=@n6
    Alpha=1
    ConstraintWeights_0=0.16
    ConstraintWeights_1:float=0.57
    ConstraintWeights_2:float=0.17
    ConstraintWeights_3:float=0.1
   ) -> (Pose) {
    alphaBoolBlend {blendOption=Linear}
    boneToModify {boneName=dyn_pillow_r_up_02_MSTR}
    constraintSetup [{targetBone {boneName=spine_04}}, {targetBone {boneName=spine_05}}, {targetBone {boneName=clavicle_r}}, {targetBone {boneName=clavicle_l}}]
    constraintWeights [1,1,1,1]
   }
   n8 = "Constraint - dyn_pillow_r_dn_01_MSTR (PPPP)"!(
    ComponentPose=@n7
    Alpha=1
    ConstraintWeights_0=0.36
    ConstraintWeights_1:float=0.38
    ConstraintWeights_2:float=0.14
    ConstraintWeights_3:float=0.12
   ) -> (Pose) {
    alphaBoolBlend {blendOption=Linear}
    boneToModify {boneName=dyn_pillow_r_dn_01_MSTR}
    constraintSetup [
     {targetBone {boneName=spine_04}}
     {targetBone {boneName=spine_05}}
     {targetBone {boneName=clavicle_r}}
     {targetBone {boneName=upperarm_twist_01_r}}
    ]
    constraintWeights [1,1,1,1]
   }
   n9 = "Constraint - dyn_pillow_r_02_MSTR (PPP)"!(ComponentPose=@n8, Alpha=1, ConstraintWeights_0=0.38, ConstraintWeights_1:float=0.57, ConstraintWeights_2:float=0.05) -> (Pose) {
    alphaBoolBlend {blendOption=Linear}
    boneToModify {boneName=dyn_pillow_r_02_MSTR}
    constraintSetup [{targetBone {boneName=spine_05}}, {targetBone {boneName=clavicle_r}}, {targetBone {boneName=upperarm_twist_01_l}}]
    constraintWeights [1,1,1]
   }
   n10 = "Constraint - dyn_pillow_r_03_MSTR (PPP)"!(ComponentPose=@n9, Alpha=1, ConstraintWeights_0=0.3, ConstraintWeights_1:float=0.65, ConstraintWeights_2:float=0.05) -> (Pose) {
    alphaBoolBlend {blendOption=Linear}
    boneToModify {boneName=dyn_pillow_r_03_MSTR}
    constraintSetup [{targetBone {boneName=spine_05}}, {targetBone {boneName=clavicle_r}}, {targetBone {boneName=upperarm_twist_01_l}}]
    constraintWeights [1,1,1]
   }
   n11 = "Constraint - dyn_pillow_r_04_MSTR (PPP)"!(ComponentPose=@n10, Alpha=1, ConstraintWeights_0=0.54, ConstraintWeights_1:float=0.4, ConstraintWeights_2:float=0.06) -> (Pose) {
    alphaBoolBlend {blendOption=Linear}
    boneToModify {boneName=dyn_pillow_r_04_MSTR}
    constraintSetup [{targetBone {boneName=spine_05}}, {targetBone {boneName=clavicle_r}}, {targetBone {boneName=upperarm_twist_01_l}}]
    constraintWeights [1,1,1]
   }
   "Constraint - dyn_pillow_r_05_MSTR (PPP)"!(ComponentPose=@n11, Alpha=1, ConstraintWeights_0=0.65, ConstraintWeights_1:float=0.3, ConstraintWeights_2:float=0.05) -> (Pose) {
    alphaBoolBlend {blendOption=Linear}
    boneToModify {boneName=dyn_pillow_r_05_MSTR}
    constraintSetup [{targetBone {boneName=spine_05}}, {targetBone {boneName=clavicle_r}}, {targetBone {boneName=clavicle_l}}]
    constraintWeights [1,1,1]
   }
   n12 = Get parachute_open() -> (parachute_open:double)
   n13 = Get arm_up_more_r() -> (arm_up_more_r:double)
   n14 = "Constraint - dyn_pillow_r_03_MSTR (PP)"!(
    ComponentPose=Local To Component(LocalPose=Copy Pose From Mesh() {bUseAttachedParent=true})
    Alpha=1
    ConstraintWeights_0=0.6
    ConstraintWeights_1:float=0.4
   ) -> (Pose) {
    alphaBoolBlend {blendOption=Linear}
    boneToModify {boneName=dyn_pillow_r_03_MSTR}
    constraintSetup [{targetBone {boneName=spine_05}}, {targetBone {boneName=clavicle_r}}]
    constraintWeights [1,1]
   }
   n15 = "Constraint - dyn_pillow_r_04_MSTR (PP)"!(ComponentPose=@n14, Alpha=1, ConstraintWeights_0=0.5, ConstraintWeights_1:float=0.5) -> (Pose) {
    alphaBoolBlend {blendOption=Linear}
    boneToModify {boneName=dyn_pillow_r_04_MSTR}
    constraintSetup [{targetBone {boneName=spine_05}}, {targetBone {boneName=clavicle_r}}]
    constraintWeights [1,1]
   }
   n16 = "Constraint - dyn_pillow_r_05_MSTR (PP)"!(ComponentPose=@n15, Alpha=1, ConstraintWeights_0=0.5, ConstraintWeights_1:float=0.5) -> (Pose) {
    alphaBoolBlend {blendOption=Linear}
    boneToModify {boneName=dyn_pillow_r_05_MSTR}
    constraintSetup [{targetBone {boneName=spine_05}}, {targetBone {boneName=clavicle_r}}]
    constraintWeights [1,1]
   }
   n17 = "Constraint - dyn_pillow_l_04_MSTR (PP)"!(ComponentPose=@n16, Alpha=1, ConstraintWeights_0=0.5, ConstraintWeights_1:float=0.5) -> (Pose) {
    alphaBoolBlend {blendOption=Linear}
    boneToModify {boneName=dyn_pillow_l_04_MSTR}
    constraintSetup [{targetBone {boneName=spine_05}}, {targetBone {boneName=clavicle_l}}]
    constraintWeights [1,1]
   }
   n18 = "Constraint - dyn_pillow_l_03_MSTR (PP)"!(ComponentPose=@n17, Alpha=1, ConstraintWeights_0=0.5, ConstraintWeights_1:float=0.5) -> (Pose) {
    alphaBoolBlend {blendOption=Linear}
    boneToModify {boneName=dyn_pillow_l_03_MSTR}
    constraintSetup [{targetBone {boneName=spine_05}}, {targetBone {boneName=clavicle_l}}]
    constraintWeights [1,1]
   }
   n19 = "Constraint - dyn_pillow_l_02_MSTR (PP)"!(ComponentPose=@n18, Alpha=1, ConstraintWeights_0=0.6, ConstraintWeights_1:float=0.4) -> (Pose) {
    alphaBoolBlend {blendOption=Linear}
    boneToModify {boneName=dyn_pillow_l_02_MSTR}
    constraintSetup [{targetBone {boneName=spine_05}}, {targetBone {boneName=clavicle_l}}]
    constraintWeights [1,1]
   }
   n20 = "Transform (Modify) Bone - Bone: dyn_head_MSTR"(ComponentPose=@n19, Alpha=1, Translation=Get ShardFloatA() -> (ShardFloatA:Vector), Rotation="0,0,0", Scale="1,1,1") {
    rotationMode=BMM_Additive
    rotationSpace=BCS_BoneSpace
    translationMode=BMM_Additive
    translationSpace=BCS_BoneSpace
    alphaBoolBlend {blendOption=Linear}
    boneToModify {boneName=dyn_head_MSTR}
   }
   n21 = RigidBody(
    ComponentPose=@n20
    ExternalForce=Break Fort Anim Input Skydiving External Force(FortAnimInput_SkydivingExternalForce=Get SkydivingExternalForceInput()) -> (
     bUseSkydivingVectorForce=~
     SkydivingVectorForce=~
     HeadToPelvisDirection=~
     FloatingMultiplier=~
     FloatingAdditive=~
     DivingMultiplier=~
     DivingAdditive=~
     ParachutingMultiplier=~
     ParachutingAdditive=~
    ).SkydivingVectorForce
   ) {
    bClampLinearTranslationLimitToRefPose=true
    bOverrideWorldGravity=true
    overridePhysicsAsset=&/M_MED_Sleepy_Time_Physics
    simulationSpace=BaseBoneSpace
    alphaBoolBlend {blendOption=Linear}
    alphaScaleBiasClamp {bInterpResult=true, interpSpeedDecreasing=15, interpSpeedIncreasing=15}
    baseBoneRef {boneName=pelvis}
    componentAppliedLinearAccClamp=<3000 3000 3000>
    componentLinearAccScale=<0.5 0.3 0.3>
    componentLinearVelScale=<0.2 0.2 0.2>
    overrideWorldGravity {z=-700}
   }
   n22 = Get rigid alpha() -> (rigid alpha:double)
   n23 = RigidBody(ComponentPose=@n21, Alpha=@n22) {
    bClampLinearTranslationLimitToRefPose=true
    bOverrideWorldGravity=true
    overridePhysicsAsset=&/M_MED_Sleepy_Time_Dyn_collar
    simulationSpace=BaseBoneSpace
    alphaBoolBlend {blendOption=Linear}
    alphaScaleBiasClamp {bInterpResult=true, interpSpeedDecreasing=15, interpSpeedIncreasing=15}
    baseBoneRef {boneName=spine_02}
    componentAppliedLinearAccClamp=<3000 3000 3000>
    componentLinearAccScale=<0.3 0.3 0.3>
    componentLinearVelScale=<0.3 0.3 0.3>
    overrideWorldGravity {z=-980}
   }
   n24 = RigidBody(ComponentPose=@n23, Alpha=@n22) {
    bClampLinearTranslationLimitToRefPose=true
    bOverrideWorldGravity=true
    overridePhysicsAsset=&/M_MED_Sleepy_Time_Dyn_arm_l
    simulationSpace=BaseBoneSpace
    alphaBoolBlend {blendOption=Linear}
    alphaScaleBiasClamp {bInterpResult=true, interpSpeedDecreasing=15, interpSpeedIncreasing=15}
    baseBoneRef {boneName=upperarm_twist_01_l}
    componentAppliedLinearAccClamp=<3000 3000 3000>
    componentLinearAccScale=<0.3 0.3 0.3>
    componentLinearVelScale=<0.2 0.2 0.2>
    overrideWorldGravity {z=-980}
   }
   n25 = RigidBody(ComponentPose=@n24, Alpha=@n22) {
    bClampLinearTranslationLimitToRefPose=true
    bOverrideWorldGravity=true
    overridePhysicsAsset=&/M_MED_Sleepy_Time_Dyn_arm_r
    simulationSpace=BaseBoneSpace
    alphaBoolBlend {blendOption=Linear}
    alphaScaleBiasClamp {bInterpResult=true, interpSpeedDecreasing=15, interpSpeedIncreasing=15}
    baseBoneRef {boneName=upperarm_twist_01_r}
    componentAppliedLinearAccClamp=<3000 3000 3000>
    componentLinearAccScale=<0.3 0.3 0.3>
    componentLinearVelScale=<0.2 0.2 0.2>
    overrideWorldGravity {z=-980}
   }
   n26 = Get arm_up_r() -> (arm_up_r:double)
   n27 = "Transform (Modify) Bone - Bone: dyn_pillow_r_00_MSTR"(ComponentPose=@n25, Alpha=@n26, Translation="1.4,0,-2.200", Rotation="0,0,0", Scale="0.91,1.000,0.670") {
    rotationMode=BMM_Additive
    rotationSpace=BCS_BoneSpace
    scaleMode=BMM_Additive
    scaleSpace=BCS_BoneSpace
    translationMode=BMM_Additive
    translationSpace=BCS_BoneSpace
    alphaBoolBlend {blendOption=Linear}
    alphaScaleBiasClamp {bInterpResult=true, interpSpeedDecreasing=15, interpSpeedIncreasing=15}
    boneToModify {boneName=dyn_pillow_r_00_MSTR}
    scale {x=0.91, z=0.67}
    translation {x=1.4, z=-2.2}
   }
   n28 = "Transform (Modify) Bone - Bone: dyn_pillow_r_01_MSTR"(ComponentPose=@n27, Alpha=@n26, Translation="4,0.000,-1.3", Rotation="0,0,0", Scale="1.000,1.000,0.95") {
    rotationMode=BMM_Additive
    rotationSpace=BCS_BoneSpace
    scaleMode=BMM_Additive
    scaleSpace=BCS_BoneSpace
    translationMode=BMM_Additive
    translationSpace=BCS_BoneSpace
    alphaBoolBlend {blendOption=Linear}
    alphaScaleBiasClamp {bInterpResult=true, interpSpeedDecreasing=15, interpSpeedIncreasing=15}
    boneToModify {boneName=dyn_pillow_r_01_MSTR}
    scale {z=0.95}
    translation {x=4, z=-1.3}
   }
   n29 = "Transform (Modify) Bone - Bone: dyn_pillow_r_02_MSTR"(ComponentPose=@n28, Alpha=@n26, Translation="5,-0.000,0.000", Rotation="0,0,0", Scale="0.72,1.000,1.000") {
    rotationMode=BMM_Additive
    rotationSpace=BCS_BoneSpace
    scaleMode=BMM_Additive
    scaleSpace=BCS_BoneSpace
    translationMode=BMM_Additive
    translationSpace=BCS_BoneSpace
    alphaBoolBlend {blendOption=Linear}
    alphaScaleBiasClamp {bInterpResult=true, interpSpeedDecreasing=15, interpSpeedIncreasing=15}
    boneToModify {boneName=dyn_pillow_r_02_MSTR}
    scale {x=0.72}
    translation {x=5}
   }
   n30 = "Transform (Modify) Bone - Bone: dyn_pillow_r_03_MSTR"(ComponentPose=@n29, Alpha=@n26, Translation="1.8,0.000,0.000", Rotation="0,0,0", Scale="1,1,1") {
    rotationMode=BMM_Additive
    rotationSpace=BCS_BoneSpace
    scaleMode=BMM_Additive
    scaleSpace=BCS_BoneSpace
    translationMode=BMM_Additive
    translationSpace=BCS_BoneSpace
    alphaBoolBlend {blendOption=Linear}
    alphaScaleBiasClamp {bInterpResult=true, interpSpeedDecreasing=15, interpSpeedIncreasing=15}
    boneToModify {boneName=dyn_pillow_r_03_MSTR}
    translation {x=1.8}
   }
   n31 = "Transform (Modify) Bone - Bone: dyn_pillow_r_00_MSTR"(ComponentPose=@n30, Alpha=@n13, Translation="4,-0.82,-1.000", Rotation="0.000,5.000,0.000", Scale="0.75,1.000,1.000") {
    rotationMode=BMM_Additive
    rotationSpace=BCS_BoneSpace
    scaleMode=BMM_Additive
    scaleSpace=BCS_BoneSpace
    translationMode=BMM_Additive
    translationSpace=BCS_BoneSpace
    alphaBoolBlend {blendOption=Linear}
    alphaScaleBiasClamp {bInterpResult=true, interpSpeedDecreasing=15, interpSpeedIncreasing=15}
    boneToModify {boneName=dyn_pillow_r_00_MSTR}
    rotation {yaw=5}
    scale {x=0.75}
    translation=<4 -0.82 -1>
   }
   n32 = "Transform (Modify) Bone - Bone: dyn_pillow_r_01_MSTR"(ComponentPose=@n31, Alpha=@n13, Translation="10,-0.463,-4.5", Rotation="0.000,-10.000,0.000", Scale="0.67,1.000,0.173") {
    rotationMode=BMM_Additive
    rotationSpace=BCS_BoneSpace
    scaleMode=BMM_Additive
    scaleSpace=BCS_BoneSpace
    translationMode=BMM_Additive
    translationSpace=BCS_BoneSpace
    alphaBoolBlend {blendOption=Linear}
    alphaScaleBiasClamp {bInterpResult=true, interpSpeedDecreasing=15, interpSpeedIncreasing=15}
    boneToModify {boneName=dyn_pillow_r_01_MSTR}
    rotation {yaw=-10.0001}
    scale {x=0.67, z=0.173}
    translation=<10 -0.463 -4.5>
   }
   n33 = "Transform (Modify) Bone - Bone: dyn_pillow_r_02_MSTR"(ComponentPose=@n32, Alpha=@n13, Translation="8,-8.000,-3", Rotation="-25.000,0.000,-10", Scale="0.800,1.000,0.4") {
    rotationMode=BMM_Additive
    rotationSpace=BCS_BoneSpace
    scaleMode=BMM_Additive
    scaleSpace=BCS_BoneSpace
    translationMode=BMM_Additive
    translationSpace=BCS_BoneSpace
    alphaBoolBlend {blendOption=Linear}
    alphaScaleBiasClamp {bInterpResult=true, interpSpeedDecreasing=15, interpSpeedIncreasing=15}
    boneToModify {boneName=dyn_pillow_r_02_MSTR}
    rotation {pitch=-25, roll=-10}
    scale {x=0.8, z=0.4}
    translation=<8 -8 -3>
   }
   n34 = "Transform (Modify) Bone - Bone: dyn_pillow_r_03_MSTR"(ComponentPose=@n33, Alpha=@n13, Translation="5.5,-2.000,-4.270", Rotation="25.000,0.000,0.000", Scale="1,1,1") {
    rotationMode=BMM_Additive
    rotationSpace=BCS_BoneSpace
    scaleMode=BMM_Additive
    scaleSpace=BCS_BoneSpace
    translationMode=BMM_Additive
    translationSpace=BCS_BoneSpace
    alphaBoolBlend {blendOption=Linear}
    alphaScaleBiasClamp {bInterpResult=true, interpSpeedDecreasing=15, interpSpeedIncreasing=15}
    boneToModify {boneName=dyn_pillow_r_03_MSTR}
    rotation {pitch=25}
    translation=<5.5 -2 -4.27>
   }
   n35 = "Transform (Modify) Bone - Bone: dyn_shoulderpad_bk_r_01_MSTR"(ComponentPose=@n34, Alpha=@n13, Translation="0,-5,15", Rotation="0,0,-20", Scale="1,1,1") {
    rotationMode=BMM_Additive
    rotationSpace=BCS_BoneSpace
    translationMode=BMM_Additive
    translationSpace=BCS_BoneSpace
    alphaBoolBlend {blendOption=Linear}
    alphaScaleBiasClamp {bInterpResult=true, interpSpeedDecreasing=15, interpSpeedIncreasing=15}
    boneToModify {boneName=dyn_shoulderpad_bk_r_01_MSTR}
    rotation {roll=-20}
    translation {y=-5, z=15}
   }
   n36 = "Transform (Modify) Bone - Bone: dyn_pillow_l_02_up"(ComponentPose=@n35, Alpha=@n12, Translation="5,0,5", Rotation="0,0,0", Scale="1,1,1") {
    rotationMode=BMM_Additive
    rotationSpace=BCS_BoneSpace
    scaleMode=BMM_Additive
    scaleSpace=BCS_BoneSpace
    translationMode=BMM_Additive
    translationSpace=BCS_BoneSpace
    alphaBoolBlend {blendOption=Linear}
    alphaScaleBiasClamp {bInterpResult=true, interpSpeedDecreasing=15, interpSpeedIncreasing=15}
    boneToModify {boneName=dyn_pillow_l_02_up}
    translation {x=5, z=5}
   }
   n37 = "Transform (Modify) Bone - Bone: dyn_pillow_l_02_dn"(ComponentPose=@n36, Alpha=@n12, Translation="8,0,-5", Rotation="0,0,0", Scale="1,1,1") {
    rotationMode=BMM_Additive
    rotationSpace=BCS_BoneSpace
    scaleMode=BMM_Additive
    scaleSpace=BCS_BoneSpace
    translationMode=BMM_Additive
    translationSpace=BCS_BoneSpace
    alphaBoolBlend {blendOption=Linear}
    alphaScaleBiasClamp {bInterpResult=true, interpSpeedDecreasing=15, interpSpeedIncreasing=15}
    boneToModify {boneName=dyn_pillow_l_02_dn}
    translation {x=8, z=-5}
   }
   n38 = "Transform (Modify) Bone - Bone: dyn_pillow_r_03_MSTR"(ComponentPose=@n37, Alpha=@n12, Translation="10,0,0", Rotation="0,0,0", Scale="1,1,1") {
    rotationMode=BMM_Additive
    rotationSpace=BCS_BoneSpace
    scaleMode=BMM_Additive
    scaleSpace=BCS_BoneSpace
    translationMode=BMM_Additive
    translationSpace=BCS_BoneSpace
    alphaBoolBlend {blendOption=Linear}
    alphaScaleBiasClamp {bInterpResult=true, interpSpeedDecreasing=15, interpSpeedIncreasing=15}
    boneToModify {boneName=dyn_pillow_r_03_MSTR}
    translation {x=10}
   }
   n39 = "Transform (Modify) Bone - Bone: dyn_pillow_l_01_MSTR"(ComponentPose=@n38, Alpha=Get arm_up_l() -> (arm_up_l:double), Translation="0,0,0", Rotation="-20,0.000,0.000", Scale="1,1,1") {
    rotationMode=BMM_Additive
    rotationSpace=BCS_BoneSpace
    alphaBoolBlend {blendOption=Linear}
    alphaScaleBiasClamp {bInterpResult=true, interpSpeedDecreasing=15, interpSpeedIncreasing=15}
    boneToModify {boneName=dyn_pillow_l_01_MSTR}
    rotation {pitch=-20}
   }
   n40 = Get jump_fix_alpha() -> (jump_fix_alpha:double)
   n41 = "Transform (Modify) Bone - Bone: dyn_r_bk_coat_01_MSTR"(ComponentPose=@n39, Alpha=@n40, Translation="0,0,0", Rotation="30,0,0", Scale="1,1,1") {
    rotationMode=BMM_Additive
    rotationSpace=BCS_BoneSpace
    alphaBoolBlend {blendOption=Linear}
    alphaScaleBiasClamp {bInterpResult=true, interpSpeedDecreasing=2}
    boneToModify {boneName=dyn_r_bk_coat_01_MSTR}
    rotation {pitch=30}
   }
   n42 = "Transform (Modify) Bone - Bone: dyn_l_bk_coat_01_MSTR"(ComponentPose=@n41, Alpha=@n40, Translation="0,0,0", Rotation="30,0,0", Scale="1,1,1") {
    rotationMode=BMM_Additive
    rotationSpace=BCS_BoneSpace
    alphaBoolBlend {blendOption=Linear}
    alphaScaleBiasClamp {bInterpResult=true, interpSpeedDecreasing=2}
    boneToModify {boneName=dyn_l_bk_coat_01_MSTR}
    rotation {pitch=30}
   }
   Output Pose(Result=Component To Local(ComponentPose=@n42)) {layerGroup=None}
  ]}
 EventGraph {nodes [
   Event BlueprintUpdateAnimation()
   > n1 = Sequence!(execute) -> (then_0, then_1, then_2:Exec, then_3:Exec, then_4:Exec, then_5:Exec)
   n2 = BreakTransform(InTransform=GetDeltaTransformFromRefPose(self=GetOwnerSkeletalMeshComponent(), BoneName=head, BaseName=spine_05))
   > Set head_rot_back(head_rot_back:double=MapRangeClamped(Value=BreakRotator(InRot=@n2.Rotation).Yaw, InRangeA=0, InRangeB=-55)) -> (Output_Get:double)
   n3 = GetOwnerSkeletalMeshComponent()
   n4 = BreakRotator(InRot=BreakTransform(InTransform=GetDeltaTransformFromRefPose(self=@n3, BoneName=upperarm_r, BaseName=spine_05)).Rotation)
   Set arm_up_r(execute=@n1.then_1, arm_up_r:double=MapRangeClamped(Value=@n4.Yaw, InRangeA=30, InRangeB=60)) -> (Output_Get:double)
   > Set arm_up_more_r(arm_up_more_r:double=MapRangeClamped(Value=@n4.Yaw, InRangeA=60, InRangeB=120)) -> (Output_Get:double)
   > Set arm_back_r(arm_back_r:double=MapRangeClamped(Value=@n4.Yaw, InRangeA=35, InRangeB=-20)) -> (Output_Get:double)
   n5 = Get arm_up_more_r()
   > Branch(Condition="float > float"(A=@n5, B=0) {is=title K2Node_CallFunction})
   > Set arm_up_r(arm_up_r=float - float(B=@n5) {is=title K2Node_CallFunction})
   n6 = GetGameTimeInSeconds()
   n7 = float * float(A=1, B=1.1)
   n8 = float * float(A="Sin (Radians)"(A=float * float(A=@n7, B=@n6)), B=0.5)
   n9 = float * float(A=1, B=1.3)
   n10 = float * float(A="Sin (Radians)"(A=float * float(A=@n9, B=@n6)), B=0.6)
   n11 = float * float(A=1, B=1.5)
   n12 = float * float(A="Sin (Radians)"(A=float * float(A=@n11, B=@n6)), B=0.5)
   Set ShardFloatA(
    execute=@n1.then_2
    ShardFloatA:Vector=vector * integer!(A:Vector=MakeVector(X=@n12, Y=@n8, Z=@n10), B:int=Get FloatingHeadOn() -> (FloatingHeadOn:int=0)) -> (ReturnValue:Vector)
   ) -> (Output_Get:Vector)
   Set parachute_open(execute=@n1.then_3, parachute_open:double="To Float (Boolean)"(InBool=Get bIsParachuteOpened())) -> (Output_Get:double)
   > Branch(Condition=Get bIsSkyDiving())
   > Set arm_up_r(arm_up_r=0)
   > Set arm_up_more_r(arm_up_more_r=0)
   n13 = GetDeltaTransformFromRefPose(self=GetOwnerSkeletalMeshComponent(), BoneName=upperarm_l, BaseName=spine_05)
   n14 = "Absolute (Float)"(A=BreakRotator(InRot=BreakTransform(InTransform=@n13).Rotation).Yaw)
   Set arm_up_l(execute=@n1.then_4, arm_up_l:double=MapRangeClamped(Value=@n14, InRangeA=50, InRangeB=105)) -> (Output_Get:double)
   > Set arm_up_low_l(arm_up_low_l:double=MapRangeClamped(Value=@n14, InRangeA=25, InRangeB=55)) -> (Output_Get:double)
   n15 = Get bIsTargeting()
   n16 = Get bIsCrouching()
   n17 = Get bIsSkyDiving()
   Set rigid alpha(execute=@n1.then_5, rigid alpha:double=SelectFloat(A=0.3, B=1, bPickA=OR Boolean!(A=@n17, B=@n15, C:bool=@n16) -> (ReturnValue))) -> (Output_Get:double)
   n18 = Get bIsJumping()
   > Set jump_fix_alpha(jump_fix_alpha:double=SelectFloat(A=1, B=0, bPickA=OR Boolean!(A=@n18, B=@n16) -> (ReturnValue))) -> (Output_Get:double)
   Event OnParachuteOpened()
   > n19 = Delay(Duration=1.5)
   ResetDynamics(execute=[@n19.then, @n20.then, @n21.then, @n24.then], InTeleportType=TeleportPhysics)
   > ResetDynamics(InTeleportType=ResetPhysics)
   n20 = Event OnBeginSkydiving()
   n21 = Event OnEndSkydiving()
   n22 = Event OnEndCrouching()
   n23 = Get bIsCrouching()
   n24 = Branch(execute=[@n22.then, @n25.then], Condition=AND Boolean(A=Get bIsTargeting(), B=NOT Boolean(A=@n23)))
   n25 = Event OnBeginTargeting() {is=title K2Node_Event}
  ]}
]
variables [
 arm_back_r {type=number, default=0.6364}
 arm_up_l {type=number}
 arm_up_low_l {type=number}
 arm_up_more_r {type=number}
 arm_up_r {type=number}
 FloatingHeadOn {type=integer, default=1}
 head_rot_back {type=number}
 jump_fix_alpha {type=number}
 parachute_open {type=number}
 rigid alpha {type=number, default=1}
 ShardFloatA {type=Vector, default=<0.4944 -0.4975 -0.0151>}
]
